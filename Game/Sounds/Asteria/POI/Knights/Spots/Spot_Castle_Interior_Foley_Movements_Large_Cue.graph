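# Sound cue for large, distant castle interior foley one-shots at the Knights POI, using indoor ambient sound class and distant-creak attenuation.
# A weighted random picks among metal drops, creaks, metallic rattles and squeaks, with a rarer magic drop, each category randomizing its own wave variations.
# A quality-level node routes the first two levels to that random and leaves two empty.
# A root modulator varies pitch and volume.
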SoundCue /Game/Sounds/Asteria/POI/Knights/Spots/Spot_Castle_Interior_Foley_Movements_Large_Cue
root=Modulator_0
nodes [
 {id=Mixer_0, children [Random_4], properties {inputVolume [1]}}
 {id=Mixer_1, children [Random_6], properties {inputVolume [1]}}
 {id=Mixer_2, children [Random_5], properties {inputVolume [1]}}
 {id=Mixer_3, children [Random_3], properties {inputVolume [1]}}
 {id=Mixer_4, children [Random_2], properties {inputVolume [1]}}
 {id=Modulator_0, children [QualityLevel_0], properties {pitchMin=0.8, volumeMax=1.2, volumeMin=1}}
 {id=QualityLevel_0, children [Random_8,Random_8,null,null]}
 {
  id=Random_2
  children [WavePlayer_0,WavePlayer_1,WavePlayer_2,WavePlayer_3,WavePlayer_4,WavePlayer_5,WavePlayer_6,WavePlayer_7,WavePlayer_8,WavePlayer_9]
  properties {weights [1,1,1,1,1,1,1,1,1,1]}
 }
 {id=Random_3, children [WavePlayer_10, WavePlayer_11, WavePlayer_12, WavePlayer_13], properties {weights [1,1,1,1]}}
 {id=Random_4, children [WavePlayer_32, WavePlayer_33], properties {weights [1,1]}}
 {
  id=Random_5
  children [
   WavePlayer_34
   WavePlayer_35
   WavePlayer_36
   WavePlayer_37
   WavePlayer_38
   WavePlayer_39
   WavePlayer_44
   WavePlayer_45
   WavePlayer_46
   WavePlayer_47
  ]
  properties {weights [1,1,1,1,1,1,1,1,1,1]}
 }
 {id=Random_6, children [WavePlayer_40, WavePlayer_41, WavePlayer_42, WavePlayer_43], properties {weights [1,1,1,1]}}
 {id=Random_8, children [Mixer_4,Mixer_3,Mixer_2,Mixer_1,Mixer_0], properties {weights [1,1,1,1,0.1]}}
 {id=WavePlayer_0, properties {soundWaveAssetPtr=&/Spot_Castle_Big_Metal_Drop_Distant_01}}
 {id=WavePlayer_1, properties {soundWaveAssetPtr=&/Spot_Castle_Big_Metal_Drop_Distant_02}}
 {id=WavePlayer_10, properties {soundWaveAssetPtr=&/Spot_Castle_Creak_Distant_01}}
 {id=WavePlayer_11, properties {soundWaveAssetPtr=&/Spot_Castle_Creak_Distant_02}}
 {id=WavePlayer_12, properties {soundWaveAssetPtr=&/Spot_Castle_Creak_Distant_03}}
 {id=WavePlayer_13, properties {soundWaveAssetPtr=&/Spot_Castle_Creak_Distant_04}}
 {id=WavePlayer_2, properties {soundWaveAssetPtr=&/Spot_Castle_Big_Metal_Drop_Distant_03}}
 {id=WavePlayer_3, properties {soundWaveAssetPtr=&/Spot_Castle_Big_Metal_Drop_Distant_04}}
 {id=WavePlayer_32, properties {soundWaveAssetPtr=&/Spot_Castle_Distant_Magic_Drop_01}}
 {id=WavePlayer_33, properties {soundWaveAssetPtr=&/Spot_Castle_Distant_Magic_Drop_02}}
 {id=WavePlayer_34, properties {soundWaveAssetPtr=&/Spot_Castle_Metallic_Rattle_Distant_01}}
 {id=WavePlayer_35, properties {soundWaveAssetPtr=&/Spot_Castle_Metallic_Rattle_Distant_02}}
 {id=WavePlayer_36, properties {soundWaveAssetPtr=&/Spot_Castle_Metallic_Rattle_Distant_03}}
 {id=WavePlayer_37, properties {soundWaveAssetPtr=&/Spot_Castle_Metallic_Rattle_Distant_04}}
 {id=WavePlayer_38, properties {soundWaveAssetPtr=&/Spot_Castle_Metallic_Rattle_Distant_05}}
 {id=WavePlayer_39, properties {soundWaveAssetPtr=&/Spot_Castle_Metallic_Rattle_Distant_06}}
 {id=WavePlayer_4, properties {soundWaveAssetPtr=&/Spot_Castle_Big_Metal_Drop_Distant_05}}
 {id=WavePlayer_40, properties {soundWaveAssetPtr=&/Spot_Castle_Metallic_Squeak_01}}
 {id=WavePlayer_41, properties {soundWaveAssetPtr=&/Spot_Castle_Metallic_Squeak_02}}
 {id=WavePlayer_42, properties {soundWaveAssetPtr=&/Spot_Castle_Metallic_Squeak_03}}
 {id=WavePlayer_43, properties {soundWaveAssetPtr=&/Spot_Castle_Metallic_Squeak_04}}
 {id=WavePlayer_44, properties {soundWaveAssetPtr=&/Spot_Castle_Metallic_Rattle_Distant_07}}
 {id=WavePlayer_45, properties {soundWaveAssetPtr=&/Spot_Castle_Metallic_Rattle_Distant_08}}
 {id=WavePlayer_46, properties {soundWaveAssetPtr=&/Spot_Castle_Metallic_Rattle_Distant_09}}
 {id=WavePlayer_47, properties {soundWaveAssetPtr=&/Spot_Castle_Metallic_Rattle_Distant_10}}
 {id=WavePlayer_5, properties {soundWaveAssetPtr=&/Spot_Castle_Big_Metal_Drop_Distant_06}}
 {id=WavePlayer_6, properties {soundWaveAssetPtr=&/Spot_Castle_Big_Metal_Drop_Distant_07}}
 {id=WavePlayer_7, properties {soundWaveAssetPtr=&/Spot_Castle_Big_Metal_Drop_Distant_08}}
 {id=WavePlayer_8, properties {soundWaveAssetPtr=&/Spot_Castle_Big_Metal_Drop_Distant_09}}
 {id=WavePlayer_9, properties {soundWaveAssetPtr=&/Spot_Castle_Big_Metal_Drop_Distant_10}}
]
properties {
 attenuationSettings=/Game/Sounds/Attenuation/Fort_TileAmbient_OneShot_Distant_Creak_Linear_Att
 maxDistance=8700
 soundClassObject=&Fort_Ambient_Indoors_Athena
 volumeMultiplier=0.752
}
