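# Sound cue for small interior castle foley in the Knights POI, playing metallic one-shot ambience on the indoor ambient class.
# A random node picks between distant rattles (six variants) and squeaks (four variants), each group choosing one wave at equal weight.
# A quality switch feeds that pick only on its first two levels, leaving the other two empty.
# The root modulator randomizes pitch and volume, with shared less-positional one-shot attenuation applied.
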SoundCue /Game/Sounds/Asteria/POI/Knights/Spots/Spot_Castle_Interior_Foley_Movements_Small_Cue
root=Modulator_0
nodes [
 {id=Mixer_1, children [Random_6], properties {inputVolume [1]}}
 {id=Mixer_2, children [Random_5], properties {inputVolume [1]}}
 {id=Modulator_0, children [QualityLevel_0], properties {pitchMin=0.7, volumeMin=0.85}}
 {id=QualityLevel_0, children [Random_8,Random_8,null,null]}
 {
  id=Random_5
  children [WavePlayer_34, WavePlayer_35, WavePlayer_36, WavePlayer_37, WavePlayer_38, WavePlayer_39]
  properties {weights [1,1,1,1,1,1]}
 }
 {id=Random_6, children [WavePlayer_40, WavePlayer_41, WavePlayer_42, WavePlayer_43], properties {weights [1,1,1,1]}}
 {id=Random_8, children [Mixer_2,Mixer_1], properties {weights [1,1]}}
 {id=WavePlayer_34, properties {soundWaveAssetPtr=&/Spot_Castle_Metallic_Rattle_Distant_01}}
 {id=WavePlayer_35, properties {soundWaveAssetPtr=&/Spot_Castle_Metallic_Rattle_Distant_02}}
 {id=WavePlayer_36, properties {soundWaveAssetPtr=&/Spot_Castle_Metallic_Rattle_Distant_03}}
 {id=WavePlayer_37, properties {soundWaveAssetPtr=&/Spot_Castle_Metallic_Rattle_Distant_04}}
 {id=WavePlayer_38, properties {soundWaveAssetPtr=&/Spot_Castle_Metallic_Rattle_Distant_05}}
 {id=WavePlayer_39, properties {soundWaveAssetPtr=&/Spot_Castle_Metallic_Rattle_Distant_06}}
 {id=WavePlayer_40, properties {soundWaveAssetPtr=&/Spot_Castle_Metallic_Squeak_01}}
 {id=WavePlayer_41, properties {soundWaveAssetPtr=&/Spot_Castle_Metallic_Squeak_02}}
 {id=WavePlayer_42, properties {soundWaveAssetPtr=&/Spot_Castle_Metallic_Squeak_03}}
 {id=WavePlayer_43, properties {soundWaveAssetPtr=&/Spot_Castle_Metallic_Squeak_04}}
]
properties {
 attenuationSettings=/Game/Sounds/Attenuation/Fort_TileAmbient_OneShot_LessPositional_Log_Att
 maxDistance=6200
 soundClassObject=&Fort_Ambient_Indoors_Athena
 volumeMultiplier=0.501
}
