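# Sound cue for castle rat squeak one-shots in the Knights POI, played through the indoor ambient one-shot sound class.
# A quality-level node sends its first two slots to a random pick of fourteen squeaks and its last two to a pool of the first five.
# Each pool passes through a modulator for slight pitch and volume variation.
# Distance falloff reuses a bird one-shot attenuation asset.
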SoundCue /Game/Sounds/Asteria/POI/Knights/Spots/Spot_Castle_Rat_Squeak_Cue
root=QualityLevel_0
nodes [
 {id=Modulator_1, children [Random_0], properties {pitchMax=1.1, volumeMax=1.1, volumeMin=0.9}}
 {id=Modulator_2, children [Random_1], properties {pitchMax=1.1, volumeMax=1.1, volumeMin=0.9}}
 {id=QualityLevel_0, children [Modulator_1,Modulator_1,Modulator_2,Modulator_2]}
 {
  id=Random_0
  children [
   WavePlayer_5
   WavePlayer_6
   WavePlayer_7
   WavePlayer_8
   WavePlayer_9
   WavePlayer_10
   WavePlayer_11
   WavePlayer_12
   WavePlayer_25
   WavePlayer_26
   WavePlayer_27
   WavePlayer_28
   WavePlayer_29
   WavePlayer_30
  ]
  properties {weights [1,1,1,1,1,1,1,1,1,1,1,1,1,1]}
 }
 {
  id=Random_1
  children [WavePlayer_31, WavePlayer_32, WavePlayer_33, WavePlayer_34, WavePlayer_35]
  properties {bShouldExcludeFromBranchCulling=true, weights [1,1,1,1,1]}
 }
 {id=WavePlayer_10, properties {soundWaveAssetPtr=&/Spot_Castle_Rat_Squeak_06}}
 {id=WavePlayer_11, properties {soundWaveAssetPtr=&/Spot_Castle_Rat_Squeak_07}}
 {id=WavePlayer_12, properties {soundWaveAssetPtr=&/Spot_Castle_Rat_Squeak_08}}
 {id=WavePlayer_25, properties {soundWaveAssetPtr=&/Spot_Castle_Rat_Squeak_09}}
 {id=WavePlayer_26, properties {soundWaveAssetPtr=&/Spot_Castle_Rat_Squeak_10}}
 {id=WavePlayer_27, properties {soundWaveAssetPtr=&/Spot_Castle_Rat_Squeak_11}}
 {id=WavePlayer_28, properties {soundWaveAssetPtr=&/Spot_Castle_Rat_Squeak_12}}
 {id=WavePlayer_29, properties {soundWaveAssetPtr=&/Spot_Castle_Rat_Squeak_13}}
 {id=WavePlayer_30, properties {soundWaveAssetPtr=&/Spot_Castle_Rat_Squeak_14}}
 {id=WavePlayer_31, properties {soundWaveAssetPtr=&/Spot_Castle_Rat_Squeak_01}}
 {id=WavePlayer_32, properties {soundWaveAssetPtr=&/Spot_Castle_Rat_Squeak_02}}
 {id=WavePlayer_33, properties {soundWaveAssetPtr=&/Spot_Castle_Rat_Squeak_03}}
 {id=WavePlayer_34, properties {soundWaveAssetPtr=&/Spot_Castle_Rat_Squeak_04}}
 {id=WavePlayer_35, properties {soundWaveAssetPtr=&/Spot_Castle_Rat_Squeak_05}}
 {id=WavePlayer_5, properties {soundWaveAssetPtr=&Spot_Castle_Rat_Squeak_01}}
 {id=WavePlayer_6, properties {soundWaveAssetPtr=&Spot_Castle_Rat_Squeak_02}}
 {id=WavePlayer_7, properties {soundWaveAssetPtr=&Spot_Castle_Rat_Squeak_03}}
 {id=WavePlayer_8, properties {soundWaveAssetPtr=&Spot_Castle_Rat_Squeak_04}}
 {id=WavePlayer_9, properties {soundWaveAssetPtr=&Spot_Castle_Rat_Squeak_05}}
]
properties {
 attenuationSettings=&Fort_TileAmbient_OneShot_Bird_Att
 bEnableBusSends=false
 maxDistance=8005
 soundClassObject=/Game/Sounds/FortSoundClassesAndModes/Fort_Ambient_OneShots_Indoors
 volumeMultiplier=0.526
 preEffectBusSends [{}]
}
