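# Ambient spot SoundCue that plays a light chain rattle one-shot, filed under the Asteria Knights POI.
# A Random node picks among eight equally weighted slow chain rattle waves, feeding a Quality Level node and then a root Modulator.
# Only the first two quality slots connect to the Random; the other two are empty.
# It borrows a bird one-shot attenuation asset, caps concurrency at two, and sends to exterior ambient bus and reverb submix.
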SoundCue /Game/Sounds/Asteria/POI/Knights/Spots/Spot_Chain_Rattle_Light_Cue
root=Modulator_0
nodes [
 {id=Modulator_0, children [QualityLevel_0]}
 {id=QualityLevel_0, children [Random_0,Random_0,null,null]}
 {
  id=Random_0
  children [WavePlayer_0,WavePlayer_2,WavePlayer_3,WavePlayer_4,WavePlayer_5,WavePlayer_6,WavePlayer_7,WavePlayer_8]
  properties {weights [1,1,1,1,1,1,1,1]}
 }
 {id=WavePlayer_0, properties {soundWaveAssetPtr=/Game/Sounds/Fort_Starlight/S1/SoundWaves/SpotFX_ChainRattle_Slow_01}}
 {id=WavePlayer_2, properties {soundWaveAssetPtr=/Game/Sounds/Fort_Starlight/S1/SoundWaves/SpotFX_ChainRattle_Slow_02}}
 {id=WavePlayer_3, properties {soundWaveAssetPtr=/Game/Sounds/Fort_Starlight/S1/SoundWaves/SpotFX_ChainRattle_Slow_03}}
 {id=WavePlayer_4, properties {soundWaveAssetPtr=/Game/Sounds/Fort_Starlight/S1/SoundWaves/SpotFX_ChainRattle_Slow_04}}
 {id=WavePlayer_5, properties {soundWaveAssetPtr=/Game/Sounds/Fort_Starlight/S1/SoundWaves/SpotFX_ChainRattle_Slow_05}}
 {id=WavePlayer_6, properties {soundWaveAssetPtr=/Game/Sounds/Fort_Starlight/S1/SoundWaves/SpotFX_ChainRattle_Slow_06}}
 {id=WavePlayer_7, properties {soundWaveAssetPtr=/Game/Sounds/Fort_Starlight/S1/SoundWaves/SpotFX_ChainRattle_Slow_07}}
 {id=WavePlayer_8, properties {soundWaveAssetPtr=/Game/Sounds/Fort_Starlight/S1/SoundWaves/SpotFX_ChainRattle_Slow_08}}
]
properties {
 attenuationSettings=&Fort_TileAmbient_OneShot_Bird_Att
 bOverrideConcurrency=true
 maxDistance=8005
 soundClassObject=&Fort_Ambient_Indoors_Athena
 volumeMultiplier=0.551
 attenuationOverrides {
  bAttenuateWithLPF=true
  distanceAlgorithm=NaturalSound
  falloffDistance=2000
  lPFFrequencyAtMax=1000
  lPFRadiusMax=3000
  lPFRadiusMin=1000
  stereoSpread=2000
  attenuationShapeExtents {x=1500}
 }
 busSends [{audioBus=null, minSendLevel=1, sendLevel=1, soundSourceBus=&AmbientAudio_Exterior_SourceBus}]
 concurrencyOverrides {maxCount=2}
 soundSubmixSends [{maxSendDistance=4000, minSendDistance=3000, minSendLevel=0.6, soundSubmix=&Fortnite_Reverb_Default}]
}
