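# Sound cue for indoor ambient one-shot creaks, named as a stone interior spot sound for the Knights POI in Asteria.
# A Random node picks among six equally weighted stereo wood creak and movement waves drawn from house, church and tree recordings.
# A Modulator at the root varies pitch and volume on each play.
# The cue uses the indoor ambient one-shot sound class, a reduced volume and its own attenuation override.
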SoundCue /Game/Sounds/Asteria/POI/Knights/Spots/Spot_Stone_Interior_Foley_Drops_Creaks_Cue
root=Modulator_0
nodes [
 {id=Modulator_0, children [Random_0], properties {pitchMin=0.8, volumeMin=0.8}}
 {id=Random_0, children [WavePlayer_0,WavePlayer_7,WavePlayer_2,WavePlayer_1,WavePlayer_5,WavePlayer_4], properties {weights [1,1,1,1,1,1]}}
 {id=WavePlayer_0, properties {soundWaveAssetPtr=/Game/Sounds/Trans_AmbientNonLoop/Wood/CreakHouseA001_Stereo}}
 {id=WavePlayer_1, properties {soundWaveAssetPtr=/Game/Sounds/Trans_AmbientNonLoop/Wood/Movement_ChurchA002_Stereo}}
 {id=WavePlayer_2, properties {soundWaveAssetPtr=/Game/Sounds/Creative/Ambiences/Wood/Movement_ChurchA007_Stereo}}
 {id=WavePlayer_4, properties {soundWaveAssetPtr=/Game/Sounds/Trans_AmbientNonLoop/Wood/Movement_ChurchB004_Stereo}}
 {id=WavePlayer_5, properties {soundWaveAssetPtr=/Game/Sounds/Trans_AmbientNonLoop/Wood/TreeCreakLong005_Stereo}}
 {id=WavePlayer_7, properties {soundWaveAssetPtr=/Game/Sounds/Trans_AmbientNonLoop/Wood/TreeCreakLong015_Stereo}}
]
properties {
 bOverrideAttenuation=true
 maxDistance=4000
 soundClassObject=/Game/Sounds/FortSoundClassesAndModes/Fort_Ambient_OneShots_Indoors
 volumeMultiplier=0.376
}
